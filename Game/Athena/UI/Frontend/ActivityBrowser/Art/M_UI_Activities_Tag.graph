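# Translucent UI material for an Activity Browser tag in the frontend, drawing a procedural frame shape with a tintable RGB color and no textures.
# An outer rectangle mask is built from centered UV distances, with parameterized offsets and edge sharpness per axis.
# An inner rectangle mask, whose axes rotate separately about a UV corner for a skewed look, cuts out the outer one.
# AlphaOverride scales the resulting mask.
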Material /Game/Athena/UI/Frontend/ActivityBrowser/Art/M_UI_Activities_Tag
graph {
 properties {
  bUsedWithStaticMesh=false
  blendMode=BLEND_Translucent
  materialDomain=MD_UI
  pixelDepthOffsetMode=PDOM_Legacy
  usedShadingModels=Default Lit
 }
 nodes [
  n1 = TextureCoordinate()
  n2 = Add(A=@n1) {constB=-0.5}
  n3 = OneMinus(None=ComponentMask(None=Abs(None=@n2)) {r=true})
  n4 = ScalarParameter() {defaultValue=0.5, parameterName=Outer-U-Offset}
  n5 = Saturate(None=Multiply(A=Subtract(A=@n3, B=@n4), B=ScalarParameter() {defaultValue=32, parameterName=Outer-U-Edge}))
  n6 = Add(A=@n1) {constB=-0.5}
  n7 = OneMinus(None=ComponentMask(None=Abs(None=@n6)) {g=true})
  n8 = ScalarParameter() {defaultValue=0.5, parameterName=Outer-V-Offset}
  n9 = Saturate(None=Multiply(A=Subtract(A=@n7, B=@n8), B=ScalarParameter() {defaultValue=48, parameterName=Outer-V-Edge}))
  n10 = TextureCoordinate()
  n11 = Add(A=Rotator(Coordinate=@n10, Time=ScalarParameter() {defaultValue=-0.012, parameterName=U-Angle}) {centerX=1, centerY=0, speed=6.2832}) {constB=-0.525}
  n12 = OneMinus(None=ComponentMask(None=Abs(None=@n11)) {r=true})
  n13 = ScalarParameter() {defaultValue=0.65, parameterName=U-Offset}
  n14 = Saturate(None=Multiply(A=Subtract(A=@n12, B=@n13), B=ScalarParameter() {defaultValue=24, parameterName=U-Edge}))
  n15 = Add(A=Rotator(Coordinate=@n10, Time=ScalarParameter() {parameterName=V-Angle}) {centerX=1, centerY=0, speed=6.2832}) {constB=-0.5}
  n16 = OneMinus(None=ComponentMask(None=Abs(None=@n15)) {g=true})
  n17 = ScalarParameter() {defaultValue=0.625, parameterName=V-Offset}
  n18 = Saturate(None=Multiply(A=Subtract(A=@n16, B=@n17), B=ScalarParameter() {defaultValue=32, parameterName=V-Edge}))
  Multiply(
   A=LinearInterpolate(A=DotProduct(A=@n5, B=@n9), Alpha=DotProduct(A=@n14, B=@n18)) {constA=1, constB=0}
   B=ScalarParameter() {defaultValue=1, parameterName=AlphaOverride}
  )
  VectorParameter() {parameterName=RGB, defaultValue=<rgba 1 1 1 1>}
 ]
}
