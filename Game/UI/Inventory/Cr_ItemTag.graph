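# Cr_ItemTag is an inventory UI tag widget showing an uppercase label in a sheared, material-backed CommonBorder.
# PreConstruct and Construct both run HandleStyling, then Update with the stored TagText.
# HandleStyling picks TagColorBeta or TagColor from IsCreativeBetaTag and applies it to both border brush and text.
# Update stores the supplied text in TagText and writes it to the label.
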WidgetBlueprint /Game/UI/Inventory/Cr_ItemTag
graphs [
 EventGraph {nodes [
   n1 = Event PreConstruct()
   HandleStyling(execute=[@n1.then, @n2.then])
   > Update!(execute, self, inTagText:Text=Get TagText() -> (TagText:Text)) -> (then)
   n2 = Event Construct()
  ]}
 HandleStyling {nodes [
   HandleStyling() {is=title K2Node_FunctionEntry}
   n1 = SelectColor(
    A=Get TagColorBeta() -> (TagColorBeta:Linear Color Structure)
    B=Get TagColor() -> (TagColor:Linear Color Structure)
    bPickA=Get IsCreativeBetaTag() -> (IsCreativeBetaTag:bool)
   )
   > SetBrushColor(self=Get BorderTagContainer() -> (BorderTagContainer:Common Border Object Reference), InBrushColor=@n1)
   > SetColorAndOpacity(self=Get Text_TagName() -> (Text_TagName:Common Text Object Reference), InColorAndOpacity=Make Slate Color(SpecifiedColor=@n1))
  ]}
 Update {nodes [
   n1 = Update!() -> (then, inTagText:Text) {is=title K2Node_FunctionEntry}
   > n2 = Set TagText(TagText:Text=@n1) -> (Output_Get:Text)
   > "SetText (Text)"(self=Get Text_TagName(), InText=@n2)
  ]}
]
variables [
 IsCreativeBetaTag {type=boolean}
 TagColor {type=LinearColor, default=<rgba 0 1 1 1>}
 TagColorBeta {type=LinearColor, default=<rgba 1 1 0 1>}
 TagText {type=string, default="{TAG}"}
]
widgets {
 info {inheritedWidgetCount=0, parentClass=&CommonUserWidget, rootWidgetClass=&CommonBorder, widgetCount=2}
 widgets [
  BorderTagContainer {
   widgetClassPath=&CommonBorder
   bIsVariable=true
   widget=BorderTagContainer
   properties {
    horizontalAlignment=HAlign_Left
    verticalAlignment=VAlign_Top
    background {
     bIsDynamicallyLoaded=false
     drawAs=Box
     imageType=NoImage
     mirroring=NoMirror
     resourceName=None
     resourceObject=/Game/Athena/UI/Frontend/ActivityBrowser/Art/M_UI_Activities_Tag
     tiling=NoTile
     imageSize=<32 32>
     margin=<ltrb 0.5 0 0.5 0>
     outlineSettings {
      bUseBrushTransparency=false
      roundingType=HalfHeightRadius
      width=0
      color {colorUseRule=UseColor_Specified, specifiedColor=<rgba 0 0 0 0>}
      cornerRadii=<0 0 0 0>
     }
     tintColor {colorUseRule=UseColor_Specified, specifiedColor=<rgba 1 1 1 1>}
     uVRegion {bIsValid=false, max=<0 0>, min=<0 0>}
    }
    padding=<ltrb 12 4 10 4>
    renderTransform {angle=0, scale=<1 1>, shear=<-8 0>, translation=<0 0>}
   }
  }
  Text_TagName {
   widgetClassPath=&CommonTextBlock
   bIsVariable=true
   parent=BorderTagContainer
   widget=Text_TagName
   properties {
    text="[tag]"
    textTransformPolicy=ToUpper
    font {
     bForceMonospaced=false
     bMaterialIsStencil=false
     fontMaterial=null
     fontObject=&BurbankSmall
     letterSpacing=0
     monospacedWidth=1
     size=15
     skewAmount=0
     typefaceFontName=Bold
     outlineSettings {
      bApplyOutlineToDropShadows=false
      bMiteredCorners=false
      bSeparateFillAlpha=false
      outlineMaterial=null
      outlineSize=0
      outlineColor=<rgba 0 0 0 1>
     }
    }
    shadowOffset=<0 0>
   }
   slot {class=BorderSlot, properties {horizontalAlignment=HAlign_Left, verticalAlignment=VAlign_Top, padding=<ltrb 12 4 10 4>}}
  }
 ]
}
