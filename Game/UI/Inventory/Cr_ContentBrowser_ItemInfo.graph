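# Item info panel for the Creative content browser, showing a name, type, description and tag chips.
# Update events fill it from raw text, a level instance or a Fort item, whose type line collapses for consumables.
# Item tags rebuild the wrap box; beta tags, including a relocated Open Beta, map to labels and append afterward.
# The wrap box and each tag are Y-flipped, so rows stack upward with upright tags.
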WidgetBlueprint /Game/UI/Inventory/Cr_ContentBrowser_ItemInfo
graphs [
 AppendBetaTag {nodes [
   n1 = AppendBetaTag!() -> (then, TargetArray:Array of Strings, ItemToFind:"String (by ref)", TextToAdd:"Text (by ref)") {is=title K2Node_FunctionEntry}
   > Branch(Condition=Contains Item(TargetArray:Array of Strings=@n1, ItemToFind:"String (by ref)"=@n1.ItemToFind))
   > Add(TargetArray:Array of Texts=Get CreativeBetaTags() -> (CreativeBetaTags:Array of Texts), NewItem:"Text (by ref)"=@n1.TextToAdd) {is=title K2Node_CallArrayFunction}
  ]}
 EventGraph {nodes [
   n1 = Event ShowDescription!() -> (OutputDelegate, then, bShow:bool) {is=title K2Node_Event}
   > SetVisibility(
    self=Get Text_Description() -> (Text_Description:Common Text Object Reference)
    InVisibility=Select(Option 0=Collapsed, Option 1=HitTestInvisible, Index=@n1.bShow)
   )
   n2 = Event UpdateWithText!() -> (OutputDelegate, then, DisplayName:"Text (by ref)", TypeName:"Text (by ref)", Description:"Text (by ref)") {is=title K2Node_Event}
   > ResetVisibilities()
   > n3 = Sequence!(execute) -> (then_0, then_1, then_2:Exec)
   n4 = Get ItemDisplayName() -> (ItemDisplayName:Common Text Object Reference)
   > SetVisibility(self=@n4, InVisibility=Select(Option 0=SelfHitTestInvisible, Option 1=Collapsed, Index=TextIsEmpty(InText=@n2.DisplayName)))
   > "SetText (Text)"(self=@n4, InText=@n2.DisplayName)
   n5 = Get TypeTextBlock() -> (TypeTextBlock:Common Text Object Reference)
   SetVisibility(execute=@n3.then_1, self=@n5, InVisibility=Select(Option 0=SelfHitTestInvisible, Option 1=Collapsed, Index=TextIsEmpty(InText=@n2.TypeName)))
   > "SetText (Text)"(self=@n5, InText=@n2.TypeName)
   n6 = Get Text_Description()
   SetVisibility(
    execute=@n3.then_2
    self=@n6
    InVisibility=Select(Option 0=SelfHitTestInvisible, Option 1=Collapsed, Index=TextIsEmpty(InText=@n2.Description))
   )
   > "SetText (Text)"(self=@n6, InText=@n2.Description)
   > PopulateTags!(execute, self, InBetaTags:bool=false, inTagArray:Array of Texts=Get EmptyTags() -> (EmptyTags:Array of Texts)) -> (then)
   Event Construct()
   > "SetText (Text)"(self=[Get ItemDisplayName(), Get TypeTextBlock()], InText=GetEmptyText!() -> (ReturnValue:Text))
   n7 = Event UpdateWithLevel!() -> (OutputDelegate, then, LevelInstanceElement:"Level Instance List Element Structure (by ref)") {is=title K2Node_Event}
   > ResetVisibilities()
   > "SetText (Text)"(
    self=Get ItemDisplayName()
    InText="To Text (String)"(InString=Break Level Instance List Element!(LevelInstanceListElement:"Level Instance List Element Structure (by ref)"=@n7.LevelInstanceElement) -> (Name:String, TerrainType:ELevelManagementTerrainType Enum, MaxMemory:int, CurrentMemory:int, NumPortals:int, NumRefs:int) {is=title K2Node_BreakStruct})
   )
   > SetVisibility(self=[Get TypeTextBlock(), Get Text_Description()], InVisibility=Hidden)
   > PopulateTags(InBetaTags=false, inTagArray=Get EmptyTags())
   n8 = Event UpdateWithFortItem!() -> (OutputDelegate, then, Item:Fort Item Object Reference) {is=title K2Node_Event}
   > Is Valid(InputObject=@n8.Item)
   > ResetVisibilities()
   > "SetText (Text)"(self=Get ItemDisplayName(), InText=GetDisplayName!(self:Fort Item Object Reference=@n8.Item) -> (ReturnValue:Text))
   > "SetText (Text)"(self=Get TypeTextBlock(), InText=GetItemTypeName!(self:Fort Item Object Reference=@n8.Item, bUsePlural:bool=false) -> (ReturnValue:Text))
   n9 = GetType!(self:Fort Item Object Reference=@n8.Item) -> (ReturnValue:EFortItemType Enum=WorldItem)
   > SetVisibility(
    self=Get TypeTextBlock()
    InVisibility=Select(Option 0=SelfHitTestInvisible, Option 1=Collapsed, Index="Equal (Enum)"(A:EFortItemType Enum=@n9, B:EFortItemType Enum=Consumable))
   )
   n10 = GetDescription!(self:Fort Item Object Reference=@n8.Item) -> (ReturnValue:Text)
   n11 = Get Text_Description()
   > SetVisibility(self=@n11, InVisibility=Select(Option 0=SelfHitTestInvisible, Option 1=Collapsed, Index=TextIsEmpty(InText=@n10)))
   > "SetText (Text)"(self=@n11, InText=@n10)
   n12 = GetItemDefinition!(self:Fort Item Object Reference=@n8.Item) -> (ReturnValue:Fort Item Definition Object Reference)
   > n13 = GetUserVisibleTagsForItem!(execute, ItemDefinition:Item Definition Base Object Reference=@n12) -> (then, ReturnValue:Array of Texts)
   > n14 = Set ItemVisibleTags(ItemVisibleTags:Array of Texts=@n13) -> (Output_Get:Array of Texts)
   > n15 = Branch(Condition=Contains Item(TargetArray:Array of Texts=@n14, ItemToFind:"Text (by ref)"=Get CreativeOpenBetaTag() -> (CreativeOpenBetaTag:Text)))
   > Remove Item!(
    execute
    TargetArray:Array of Texts=Get ItemVisibleTags() -> (ItemVisibleTags:Array of Texts)
    Item:"Text (by ref)"=Get CreativeOpenBetaTag()
   ) -> (then, ReturnValue:bool)
   > n16 = Add(TargetArray:Array of Strings=Get BetaTagStrings() -> (BetaTagStrings:Array of Strings), NewItem:"String (by ref)"=OpenBeta)
   PopulateTags(execute=[@n16.then, @n15.else], InBetaTags=false, inTagArray=Get ItemVisibleTags())
   n17 = GetCreativeBetaPermissionTags!(ItemDefinition:Item Definition Base Object Reference=@n12) -> (ReturnValue:Array of Strings)
   > n18 = Branch(Condition=Is Array Non Empty!(InputArray:Array of Strings=@n17) -> (Non-Empty:bool=~) {is=title K2Node_MacroInstance})
   > n19 = Append Array!(execute, TargetArray=Get BetaTagStrings(), SourceArray:Array of Strings=@n17) -> (then)
   Branch(execute=[@n19.then, @n18.else], Condition=Is Array Non Empty(InputArray=Get BetaTagStrings()) -> (Non-Empty=~))
   > PopulateBetaTags!(execute, self, BetaTagStrings:Array of Strings=Get BetaTagStrings()) -> (then)
  ]}
 PopulateBetaTags {nodes [
   n1 = PopulateBetaTags!() -> (then, BetaTagStrings:Array of Strings) {is=title K2Node_FunctionEntry}
   > Clear(TargetArray:Array of Texts=Get CreativeBetaTags())
   > n2 = Keys!(execute, TargetMap:Map of Strings to Texts=Get CreativeBetaTagsMap() -> (CreativeBetaTagsMap:Map of Strings to Texts)) -> (then, Keys:Array of Strings)
   > n3 = For Each Loop(Array=@n2)
   > AppendBetaTag!(
    execute
    self
    TargetArray:Array of Strings=@n1
    ItemToFind:"String (by ref)"=@n3
    TextToAdd:"Text (by ref)"=Find(TargetMap:Map of Strings to Texts=Get CreativeBetaTagsMap(), Key=@n3) -> (Value:Text)
   ) -> (then) {is=title K2Node_CallFunction}
   PopulateTags(execute=@n3.Completed, InBetaTags=true, inTagArray=Get CreativeBetaTags())
  ]}
 PopulateTags {nodes [
   n1 = PopulateTags!() -> (then, InBetaTags:bool=~, inTagArray:Array of Texts) {is=title K2Node_FunctionEntry}
   > n2 = Branch(Condition=@n1)
   Branch(
    execute=[@n2.then, @n5.then]
    Condition="integer > integer"(A=Length(TargetArray:Array of Texts=@n1.inTagArray), B=0) {is=title K2Node_CallFunction}
   )
   > n3 = For Each Loop(Array:Array of Texts=@n1.inTagArray) -> (Array Element:Text)
   > n4 = Create Widget!(
    execute:Exec
    Class:User Widget Class Reference=&/Cr_ItemTag.Cr_ItemTag_C
    OwningPlayer:Player Controller Object Reference=GetOwningPlayer()
    TagText:Text=@n3
    IsCreativeBetaTag:bool=@n1
   ) -> (then:Exec, ReturnValue:Cr Item Tag Object Reference)
   > SetRenderScale!(execute, self:Widget Object Reference=@n4, Scale:Vector 2D Structure=(X=1,Y=-1)) -> (then)
   > AddChildToWrapBox!(execute, self:Wrap Box Object Reference=Get Wbox_Tags() -> (Wbox_Tags:Wrap Box Object Reference), Content:Widget Object Reference=@n4) -> (then, ReturnValue:Wrap Box Slot Object Reference)
   n5 = ClearChildren(execute=@n2.else, self=Get Wbox_Tags())
  ]}
 ResetVisibilities {nodes [
   ResetVisibilities() {is=title K2Node_FunctionEntry}
   > SetVisibility(self=Get ItemDisplayName(), InVisibility=SelfHitTestInvisible)
   > SetVisibility(self=Get TypeTextBlock(), InVisibility=SelfHitTestInvisible)
   > Clear(TargetArray:Array of Strings=Get BetaTagStrings())
   > Clear(TargetArray:Array of Texts=Get ItemVisibleTags())
  ]}
]
variables [
 BetaTagStrings {type="array<string>", default []}
 CreativeBetaTags {type="array<string>", default []}
 CreativeBetaTagsMap {type="map<string>", default {Developer=Developer, EarlyAccess=EarlyAccess, OpenBeta=Open Beta, Partner=Partner}}
 CreativeOpenBetaTag {type=string, default=Open Beta}
 EmptyTags {type="array<string>", default []}
 ItemVisibleTags {type="array<string>", default []}
]
widgets {
 info {inheritedWidgetCount=0, parentClass=/Script/FortniteUI.CreativeContentBrowserItemInfo, rootWidgetClass=&SizeBox, widgetCount=12}
 widgets [
  SizeBox_Body {
   widgetClassPath=&SizeBox
   bIsVariable=true
   widget=SizeBox_Body
   properties {bOverride_MinDesiredHeight=true, heightOverride=168, minDesiredHeight=104, minDesiredWidth=400, widthOverride=660}
  }
  Overlay_0 {widgetClassPath=&Overlay, parent=SizeBox_Body, widget=Overlay_0, slot {class=SizeBoxSlot}}
  VerticalBox_0 {
   widgetClassPath=&VerticalBox
   parent=Overlay_0
   widget=VerticalBox_0
   slot {class=OverlaySlot, properties {horizontalAlignment=HAlign_Fill, verticalAlignment=VAlign_Fill}}
  }
  Wbox_Tags {
   widgetClassPath=&WrapBox
   bIsVariable=true
   parent=VerticalBox_0
   widget=Wbox_Tags
   properties {innerSlotPadding=<2 2>, renderTransform {angle=0, scale=<1 -1>, shear=<0 0>, translation=<0 0>}}
   slot {class=VerticalBoxSlot, properties {padding=<ltrb 12 0 0 4>, size {sizeRule=Fill, value=1}}}
  }
  Overlay_1 {widgetClassPath=&Overlay, parent=VerticalBox_0, widget=Overlay_1, slot {class=VerticalBoxSlot}}
  Image_Backplate {
   widgetClassPath=&Image
   bIsVariable=true
   parent=Overlay_1
   widget=Image_Backplate
   properties {brush {
     bIsDynamicallyLoaded=false
     drawAs=Image
     imageType=NoImage
     mirroring=NoMirror
     resourceName=None
     resourceObject=/Game/UI/Inventory/Art/M_UI_Creative_ContentBrowser_ItemInfoBackground
     tiling=NoTile
     imageSize=<32 32>
     margin=<ltrb 0 0 0 0>
     outlineSettings {
      bUseBrushTransparency=false
      roundingType=HalfHeightRadius
      width=0
      color {colorUseRule=UseColor_Specified, specifiedColor=<rgba 0 0 0 0>}
      cornerRadii=<0 0 0 0>
     }
     tintColor {colorUseRule=UseColor_Specified, specifiedColor=<rgba 1 1 1 1>}
     uVRegion {bIsValid=false, max=<0 0>, min=<0 0>}
    }}
   slot {class=OverlaySlot, properties {horizontalAlignment=HAlign_Fill, verticalAlignment=VAlign_Fill}}
  }
  Image_Underline {
   widgetClassPath=&Image
   bIsVariable=true
   parent=Overlay_1
   widget=Image_Underline
   properties {
    brush {
     bIsDynamicallyLoaded=false
     drawAs=Image
     imageType=NoImage
     mirroring=NoMirror
     resourceName=None
     resourceObject=null
     tiling=NoTile
     imageSize=<32 3>
     margin=<ltrb 0 0 0 0>
     outlineSettings {
      bUseBrushTransparency=false
      roundingType=HalfHeightRadius
      width=0
      color {colorUseRule=UseColor_Specified, specifiedColor=<rgba 0 0 0 0>}
      cornerRadii=<0 0 0 0>
     }
     tintColor {colorUseRule=UseColor_Specified, specifiedColor=<rgba 1 1 1 1>}
     uVRegion {bIsValid=false, max=<0 0>, min=<0 0>}
    }
    colorAndOpacity=<rgba 0.1329 0.6038 1 1>
   }
   slot {class=OverlaySlot, properties {horizontalAlignment=HAlign_Fill, verticalAlignment=VAlign_Bottom, padding=<ltrb 5 0 0 0>}}
  }
  VerticalBox_2 {
   widgetClassPath=&VerticalBox
   parent=Overlay_1
   widget=VerticalBox_2
   slot {class=OverlaySlot, properties {horizontalAlignment=HAlign_Fill, verticalAlignment=VAlign_Fill, padding=<ltrb 24 6 16 16>}}
  }
  TypeTextBlock {
   widgetClassPath=&CommonTextBlock
   bIsVariable=true
   parent=VerticalBox_2
   widget=TypeTextBlock
   properties {
    justification=Right
    scrollStyle=&TextScrollStyle-Base_C
    style=/FortUILibrary/Styles/Typography/Foundation/TextStyle-Base-XS-B-70pc.TextStyle-Base-XS-B-70pc_C
    text="{TypeName}"
    colorAndOpacity {colorUseRule=UseColor_Specified, specifiedColor=<rgba 1 1 1 0.7>}
    font {
     bForceMonospaced=false
     bMaterialIsStencil=false
     fontMaterial=null
     fontObject=&NotoSans
     letterSpacing=0
     monospacedWidth=1
     size=16
     skewAmount=0
     typefaceFontName=NotoSans-Bold
     outlineSettings {
      bApplyOutlineToDropShadows=false
      bMiteredCorners=false
      bSeparateFillAlpha=false
      outlineMaterial=null
      outlineSize=0
      outlineColor=<rgba 0 0 0 1>
     }
    }
    shadowOffset=<0 0>
   }
   slot {class=VerticalBoxSlot}
  }
  ItemDisplayName {
   widgetClassPath=&CommonTextBlock
   bIsVariable=true
   parent=VerticalBox_2
   widget=ItemDisplayName
   properties {
    clipping=ClipToBounds
    justification=Right
    scrollStyle=&TextScrollStyle-Base_C
    style=/FortUILibrary/Styles/Typography/Foundation/TextStyle-Header-S.TextStyle-Header-S_C
    text="{ItemName}"
    font {
     bForceMonospaced=false
     bMaterialIsStencil=false
     fontMaterial=null
     fontObject=&Burbank
     letterSpacing=0
     monospacedWidth=1
     size=24
     skewAmount=0
     typefaceFontName=Burbank-Black
     outlineSettings {
      bApplyOutlineToDropShadows=false
      bMiteredCorners=false
      bSeparateFillAlpha=false
      outlineMaterial=null
      outlineSize=0
      outlineColor=<rgba 0 0 0 1>
     }
    }
    shadowOffset=<0 0>
   }
   slot {class=VerticalBoxSlot}
  }
  SizeBox_0 {
   widgetClassPath=&SizeBox
   parent=VerticalBox_2
   widget=SizeBox_0
   properties {bOverride_MaxDesiredHeight=true, maxDesiredHeight=300}
   slot {class=VerticalBoxSlot}
  }
  Text_Description {
   widgetClassPath=&CommonTextBlock
   bIsVariable=true
   parent=SizeBox_0
   widget=Text_Description
   properties {
    autoWrapText=true
    clipping=ClipToBounds
    justification=Right
    style=&TextStyle-Base-XS-B-70pc_C
    text="{Description}"
    textOverflowPolicy=Ellipsis
    visibility=HitTestInvisible
    colorAndOpacity {colorUseRule=UseColor_Specified, specifiedColor=<rgba 1 1 1 0.7>}
    font {
     bForceMonospaced=false
     bMaterialIsStencil=false
     fontMaterial=null
     fontObject=&NotoSans
     letterSpacing=0
     monospacedWidth=1
     size=16
     skewAmount=0
     typefaceFontName=NotoSans-Bold
     outlineSettings {
      bApplyOutlineToDropShadows=false
      bMiteredCorners=false
      bSeparateFillAlpha=false
      outlineMaterial=null
      outlineSize=0
      outlineColor=<rgba 0 0 0 1>
     }
    }
    shadowOffset=<0 0>
   }
   slot {class=SizeBoxSlot}
  }
 ]
}
