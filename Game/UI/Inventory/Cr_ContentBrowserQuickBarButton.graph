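# Quick bar button widget for the creative content browser, wrapping a CreativeQuickbarSlot for one configurable slot index and quickbar type.
# On Construct it pushes both values into the slot and rebroadcasts slot touches as OnItemTapped with the index.
# Dropping a FortItem on it fires OnItemDropped with index, type and item.
# Event_Pulse plays a double-speed scale and opacity flash on the BarHighlight overlay image.
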WidgetBlueprint /Game/UI/Inventory/Cr_ContentBrowserQuickBarButton
animations [Pulse {
  binding_count=1
  length_seconds=1
  object=&self:Pulse
  sequence {bindings [BarHighlight {
     id=#1
     possessed_class=Image
     tracks [
      {
       type=MovieScene2DTransformTrack
       sections [{
         type=MovieScene2DTransformSection
         end_frame=20
         start_frame=0
         channels [
          Translation.X {type=MovieSceneScriptingFloatChannel}
          Translation.Y {type=MovieSceneScriptingFloatChannel}
          Angle {type=MovieSceneScriptingFloatChannel}
          Scale.X {type=MovieSceneScriptingFloatChannel, keys [[0,1],[20,1.5]]}
          Scale.Y {type=MovieSceneScriptingFloatChannel, keys [[0,1],[20,1.5]]}
          Shear.X {type=MovieSceneScriptingFloatChannel}
          Shear.Y {type=MovieSceneScriptingFloatChannel}
         ]
        }]
      }
      {
       type=MovieSceneFloatTrack
       display_name=Render Opacity
       sections [{
         type=MovieSceneFloatSection
         end_frame=20
         start_frame=0
         channels [None {type=MovieSceneScriptingFloatChannel, keys [[0,0],[10,0.2491],[15,0]]}]
        }]
      }
     ]
    }]}
 }]
graphs [
 EventGraph {nodes [
   n1 = OnSlotTapped() {is=title K2Node_CustomEvent}
   > Call On Item Tapped!(execute:Exec, self:Cr Content Browser Quick Bar Button Object Reference, NewParam:int=Get CreativeSlotIndex() -> (CreativeSlotIndex:int=0)) -> (then:Exec) {is=title K2Node_CallDelegate}
   Event_Pulse() {is=title K2Node_CustomEvent}
   > PlayAnimation(InAnimation=Get Pulse() -> (Pulse:Widget Animation Object Reference), PlaybackSpeed=2)
   Event Construct()
   > SetQuickbarIndexRuntime!(
    execute
    self:Fort Quick Bar Slot Base Object Reference=Get CreativeQuickbarSlot1() -> (CreativeQuickbarSlot1:Creative Quickbar Slot Object Reference)
    Index:int=Get CreativeSlotIndex() -> (CreativeSlotIndex=0)
    bRefresh:bool=true
   ) -> (then)
   > SetQuickbarTypeRuntime!(
    execute
    self:Fort Quick Bar Slot Base Object Reference=Get CreativeQuickbarSlot1()
    NewQuickBarType:EFortQuickBars Enum=Get QuickbarType() -> (QuickbarType:EFortQuickBars Enum=Primary)
    bRefresh:bool=true
   ) -> (then)
   > Bind Event to On Slot Touched(self:Creative Quickbar Slot Object Reference=Get CreativeQuickbarSlot1(), Delegate=@n1) {is=title K2Node_AddDelegate}
   Event PreConstruct()
   Event Tick()
  ]}
 OnDrop {nodes [
   n1 = OnDrop!() -> (then, MyGeometry:Geometry Structure, PointerEvent:Pointer Event Structure, Operation:Drag Drop Operation Object Reference) {is=title K2Node_FunctionEntry}
   > n2 = Cast To FortItem!(execute, Object=Get Payload!(self:Drag Drop Operation Object Reference=@n1.Operation) -> (Payload:Object Reference)) -> (then, CastFailed, AsFort Item:Fort Item Object Reference) {is=title K2Node_DynamicCast}
   > Call On Item Dropped!(
    execute:Exec
    self:Cr Content Browser Quick Bar Button Object Reference
    ItemSlot:int=Get CreativeSlotIndex() -> (CreativeSlotIndex=0)
    QuickbarType:EFortQuickBars Enum=Get QuickbarType() -> (QuickbarType=Primary)
    Item:Fort Item Object Reference=@n2
   ) -> (then:Exec) {is=title K2Node_CallDelegate}
   > Return Node!(execute, ReturnValue:bool=true)
  ]}
 OnItemDropped {nodes [OnItemDropped!() -> (then, ItemSlot:int, QuickbarType:EFortQuickBars Enum=Primary, Item:Fort Item Object Reference) {is=title K2Node_FunctionEntry}]}
 OnItemTapped {nodes [OnItemTapped!() -> (then, NewParam:int) {is=title K2Node_FunctionEntry}]}
]
variables [
 CreativeSlotIndex {type=integer}
 OnItemDropped {}
 OnItemTapped {}
 QuickbarType {type=EFortQuickBars, default=Creative}
]
widgets {
 info {inheritedWidgetCount=1, parentClass=&CommonButtonLegacy, rootWidgetClass=&Overlay, widgetCount=4}
 widgets [
  Overlay_0 {widgetClassPath=&Overlay, widget=Overlay_0}
  CreativeQuickbarSlot1 {
   widgetClassPath=/Game/Athena/HUD/CreativeQuickbarSlot
   bIsVariable=true
   parent=Overlay_0
   widget=CreativeQuickbarSlot1
   properties {bIsFocusable=true, bShouldTakeItemSelectionInputFromHUD=false}
   slot {class=OverlaySlot, properties {horizontalAlignment=HAlign_Center, verticalAlignment=VAlign_Center}}
  }
  BarHighlight {
   widgetClassPath=&Image
   parent=Overlay_0
   widget=BarHighlight
   properties {
    renderOpacity=0
    visibility=SelfHitTestInvisible
    brush {
     bIsDynamicallyLoaded=false
     drawAs=Image
     imageType=NoImage
     mirroring=NoMirror
     resourceName=None
     resourceObject=/Game/UI/Inventory/Art/MI_UI_CrQuickbarPing
     tiling=NoTile
     imageSize=<90 90>
     margin=<ltrb 0 0 0 0>
     outlineSettings {
      bUseBrushTransparency=false
      roundingType=HalfHeightRadius
      width=0
      color {colorUseRule=UseColor_Specified, specifiedColor=<rgba 0 0 0 0>}
      cornerRadii=<0 0 0 0>
     }
     tintColor {colorUseRule=UseColor_Specified, specifiedColor=<rgba 1 1 1 1>}
     uVRegion {bIsValid=false, max=<0 0>, min=<0 0>}
    }
    renderTransform {angle=0, scale=<1.5 1.5>, shear=<0 0>, translation=<0 20>}
   }
   slot {class=OverlaySlot, properties {horizontalAlignment=HAlign_Center, verticalAlignment=VAlign_Center}}
  }
  InputActionWidget {widgetClassPath=&CommonActionWidget, bInherited=true, widget=null}
 ]
}
